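# Animation Blueprint for the NanoChain backpack cosmetic that adds secondary motion to its dangling boltcutters.
# Two chained, angle-limited Anim Dynamics nodes simulate the root and middle boltcutter bones in root-relative space, using a gravity override computed each update by a deprecated helper.
# The event graph converts the middle bone's Euler Y delta from the body bone into an alpha.
# That alpha scales opposite additive pitch rotations on the top and bottom cutter bones, so they pivot with the swing.
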AnimBlueprint /BRCosmetics/Accessories/FORT_Backpacks/Backpack_F_MED_NanoChain/Meshes/F_MED_NanoChain_Pack_AnimBP
target_skeleton=&/F_MED_NanoChain_Pack_Skeleton
graphs [
 AnimGraph {nodes [
   n1 = Anim Dynamics - Bone: dyn_boltcutters!(ComponentPose=Local To Component(LocalPose=Input Pose()), Alpha=1, GravityOverride:Vector=Get GravityOverride(), ExternalForce="0,0,0") -> (Pose) {
    angularDampingOverride=0.94
    angularSpringConstant=800
    bAngularSpring=true
    bOverrideAngularDamping=true
    bUseGravityOverride=true
    linearDampingOverride=0.9
    simulationSpace=RootRelative
    alphaBoolBlend {blendOption=Linear}
    boundBone {boneName=dyn_boltcutters}
    componentAppliedLinearAccClamp=<3000 3000 3000>
    componentLinearAccScale=<0.3 0.3 0.3>
    componentLinearVelScale=<0.3 0.3 0.3>
    physicsBodyDefinitions [{
      boundBone {boneName=dyn_boltcutters}
      boxExtents=<15 15 15>
      constraintSetup {angularLimitsMax=<0 0 15>, angularLimitsMin=<0 0 -35>, angularTarget=<1 0 0>}
      localJointOffset=<3 0 0>
     }]
   }
   n2 = Anim Dynamics - Bone: dyn_boltcuttersMiddle_01!(ComponentPose=@n1, Alpha=1, GravityOverride:Vector=Get GravityOverride(), ExternalForce="0,0,0") -> (Pose) {
    angularDampingOverride=0.95
    angularSpringConstant=1500
    bAngularSpring=true
    bOverrideAngularDamping=true
    bUseGravityOverride=true
    linearDampingOverride=0.9
    simulationSpace=RootRelative
    alphaBoolBlend {blendOption=Linear}
    boundBone {boneName=dyn_boltcuttersMiddle_01}
    componentAppliedLinearAccClamp=<3000 3000 3000>
    componentLinearAccScale=<0.4 0.4 0.4>
    componentLinearVelScale=<0.4 0.4 0.4>
    physicsBodyDefinitions [{
      boundBone {boneName=dyn_boltcuttersMiddle_01}
      boxExtents=<18 18 18>
      constraintSetup {angularLimitsMax=<0 30 0>, angularLimitsMin=<0 0 0>, angularTarget=<1 0 0>}
      localJointOffset=<0 0 -1>
     }]
   }
   n3 = "Transform (Modify) Bone - Bone: dyn_boltcutterBottom_02"(ComponentPose=@n2, Alpha=Get boltcutterRot() -> (boltcutterRot:double), Translation="0,0,0", Rotation="8,0,0", Scale="1,1,1") {
    rotationMode=BMM_Additive
    rotationSpace=BCS_BoneSpace
    alphaBoolBlend {blendOption=Linear}
    boneToModify {boneName=dyn_boltcutterBottom_02}
    rotation {pitch=8}
   }
   n4 = "Transform (Modify) Bone - Bone: dyn_boltcuttersTop_01"(ComponentPose=@n3, Alpha=Get boltcutterRot(), Translation="0,0,0", Rotation="-8,0,0", Scale="1,1,1") {
    rotationMode=BMM_Additive
    rotationSpace=BCS_BoneSpace
    alphaBoolBlend {blendOption=Linear}
    boneToModify {boneName=dyn_boltcuttersTop_01}
    rotation {pitch=-8}
   }
   Output Pose(Result=Component To Local(ComponentPose=@n4)) {layerGroup=None}
  ]}
 EventGraph {nodes [
   Event BlueprintUpdateAnimation()
   > n1 = Sequence()
   > n2 = Gravity Override State Function DEPRECATED(CharacterPartAnimInstance=Self-Reference(), GameTimeInSeconds=GetGameTimeInSeconds(), GravityOverrideParams=Get Gravity Override Params())
   > Set GravityOverride(GravityOverride=@n2)
   n3 = Set Skel Mesh(execute=@n1.then_1, Skel Mesh=GetPartSkeletalMeshComponent())
   > Set boltcutterRot(boltcutterRot:double=GetAlphaFromReferenceTransformDelta(
     SkeletalMeshComponent=@n3
     BoneName=dyn_boltcuttersMiddle_01
     BaseName=body
     DeltaComponentType=EulerY
     InputRange=(X=0,Y=-10)
     TwistAxis="0,1,0"
    )) -> (Output_Get:double)
   TryGetPawnOwner()
  ]}
]
variables [
 boltcutterRot {type=number}
 Gravity Override Params {type=GravityOverrideParamsStruct}
 GravityOverride {type=Vector}
 Skel Mesh {type=SkeletalMeshComponent}
]
